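; Interactable button for the Wild Estate FloorMist encounter that starts a round releasing the hounds.
; BeginPlay destroys it if disabled; otherwise it lights up and, unless tagged ButtonDesk, on authority binds its destruction to its tagged glow desk.
; Interacting or ActivateOnDestroy marks the wave active, blocking reuse, kills the lights, plays sound and music stinger, and ends the discoverable gag.
; SpawnProgressBar creates a boss progress-bar ChallengeManager fed by encounter references and spawn count.
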(event Custom|turn_light%20off
  (Rendering|Material|SetVectorParameterValue (Class|BTechArtGlobalLibrary|GetDynamicMaterialInstance (Variables|Default|GetButtonMesh)) "Emissive_Color_Multiplier" "(R=0.000000,G=0.000000,B=0.000000,A=1.000000)" :!class /Script/Engine.MaterialInstanceDynamic)
  (Rendering|Material|SetScalarParameterValue (Class|BTechArtGlobalLibrary|GetDynamicMaterialInstance (Variables|Default|GetSM_Airglow)) "Brightness" 0.0 :!class /Script/Engine.MaterialInstanceDynamic)
  (Audio|Components|Audio|FadeOut (Variables|Default|GetSFX_LightOnHum) 0.5 0.0 "Linear" :!class /Script/Engine.AudioComponent))

(event Custom|Turn_Lights_On
  (Rendering|Material|SetVectorParameterValue (Class|BTechArtGlobalLibrary|GetDynamicMaterialInstance (Variables|Default|GetButtonMesh)) "Emissive_Color_Multiplier" "(R=3.000000,G=3.000000,B=3.000000,A=1.000000)" :!class /Script/Engine.MaterialInstanceDynamic)
  (Rendering|Material|SetScalarParameterValue (Class|BTechArtGlobalLibrary|GetDynamicMaterialInstance (Variables|Default|GetSM_Airglow)) "Brightness" 0.5 :!class /Script/Engine.MaterialInstanceDynamic))

(event Custom|SpawnProgressBar
  (bind (_returnvalue_location _returnvalue_rotation) (Math|Transform|BreakTransform (Transformation|GetActorTransform self)))
  (CallFunction|TurnLightOff)
  (bind _returnvalue (Game|SpawnActorfromClass "/WildEstate_FloorMist/Gameplay/ProgressBar/ChallengeManager_SimpleBossProgressBar_WildEstate_FloorMist.ChallengeManager_SimpleBossProgressBar_WildEstate_FloorMist_C" (Math|Transform|MakeTransform _returnvalue_location _returnvalue_rotation "5.000000,5.000000,3.000000") "Undefined" "MultiplyWithRoot" 0 0))
  (bind _output_get (Variables|Default|SetProgressBar _returnvalue))
  (Variables|/WildEstate_FloorMist/Gameplay/ProgressBar/ChallengeManager_SimpleBossProgressBar_WildEstate_FloorMist.ChallengeManager_SimpleBossProgressBar_WildEstate_FloorMist_C|SetSpawnsInWave (Utilities|GetRoundMaxSpawned) _output_get)
  (Variables|/WildEstate_FloorMist/Gameplay/ProgressBar/ChallengeManager_SimpleBossProgressBar_WildEstate_FloorMist.ChallengeManager_SimpleBossProgressBar_WildEstate_FloorMist_C|SetLivingWorldVolume (Variables|Default|GetEncounterLivingWorldVolume) _output_get)
  (Variables|/WildEstate_FloorMist/Gameplay/ProgressBar/ChallengeManager_SimpleBossProgressBar_WildEstate_FloorMist.ChallengeManager_SimpleBossProgressBar_WildEstate_FloorMist_C|SetLWV_Encounter (Variables|Default|GetEncounterRef) _output_get)
  (Music|AddCurrentMusicTagtoActor _output_get self)
  (Class|ChallengeManager|Init _output_get))

(event EventBeginPlay
  (bind _tag (Variables|Default|GetTag))
  (if (ScalableFloat|GetValueasBool (Variables|Default|GetEnabled))
    (CallFunction|TurnLightsOn)
    (Class|FortAsyncActionCurrentPlaylistReady|PlaylistDataReadyAsync
      (:OnReady
        (Networking|SwitchHasAuthority
          (:Authority
            (Utilities|FlowControl|DelayUntilNextTick)
            (if (!= _tag "ButtonDesk")
              (bind _outactors (Actor|GetAllActorsOfClasswithTag "/WildEstate/Environment/Props/GlowDonut/GlowDonut_OfficeDesk_A/Blueprints/BP_GlowDonut_OfficeDesk_A.BP_GlowDonut_OfficeDesk_A_C" _tag))
              (Utilities|IsValid (Utilities|Array|Get(acopy) _outactors)
                (:Is%20Valid
                  (Game|BindEventtoOnDestroyed (Utilities|Array|Get(acopy) _outactors) (EventDispatchers|CreateEvent :!event Custom|DeskActorDestroyed)))))))))
    (else
      (Actor|DestroyActor))))

(event Custom|ActivateOnDestroy
  (Variables|Default|SetWaveActive true)
  (CallFunction|TurnLightOff)
  (CallFunction|PlayAudioTo3P)
  (Class|BDiscoverableGagComponent|EndDiscoverableGagServer (Variables|Default|GetB_DiscoverableGagComponent)))

(event Interaction|EventBlueprintOnInteract
  (params (InteractingPawn :type Fort%20Pawn%20Object%20Reference) (InteractionBeingAttempted :type EInteractionBeingAttempted%20Enum :default "FirstInteraction"))
  (Variables|Default|SetWaveActive true)
  (CallFunction|TurnLightOff)
  (CallFunction|PlayAudioTo3P)
  (Class|BDiscoverableGagComponent|EndDiscoverableGagServer (Variables|Default|GetB_DiscoverableGagComponent)))

(event Custom|Play%20Audio%20to%203P
  (bind _buttonbase (Variables|Default|GetButtonBase))
  (bind _returnvalue (Transformation|GetWorldLocation _buttonbase))
  (Audio|PlaySoundatLocation "/WildEstate_FloorMist/Assets/Audio/SFX/MSS_FloorMist_Interact_Complete.MSS_FloorMist_Interact_Complete" _returnvalue)
  (bind _returnvalue_2 (Audio|SpawnSoundatLocation (Variables|Default|GetMusic%20Stinger) _returnvalue))
  (Audio|AddEventbyComponent (GameInstanceSubsystems|GetMusicEventSubsystem) _returnvalue_2 "(TagName=\"MusicEventSubsystem.Event.Gameplay\")" "(GameplayTags=)" 0 ""))

(event Custom|DeskActorDestroyed
  (params (DestroyedActor :type Actor%20Object%20Reference))
  (Actor|DestroyActor))

(fn BlueprintCanInteract
  (params (InteractingPawn :type Fort%20Pawn%20Object%20Reference) (InteractionBeingAttempted :type EInteractionBeingAttempted%20Enum :default "FirstInteraction") (InteractionType :type TInteractionType%20Enum :default "IT_NoInteraction"))
  (return (not (Variables|Default|GetWaveActive))))

(fn BlueprintGetInteractionString
  (params (InteractingPawn :type Fort%20Pawn%20Object%20Reference) (InteractionBeingAttempted :type EInteractionBeingAttempted%20Enum :default "FirstInteraction"))
  (return (Variables|Default|GetInteractString)))

(fn BlueprintGetSubInteractionString
  (params (InteractingPawn :type Fort%20Pawn%20Object%20Reference) (InteractionBeingAttempted :type EInteractionBeingAttempted%20Enum :default "FirstInteraction"))
  (if (Variables|Default|GetWaveActive)
    (return "NSLOCTEXT(\"[DD35C90E7AAB6BBDA0637D00CF2F322A]\", \"CC91D5E04D77D389BF055C92B3605EF1\", \"Round Active, The Hounds are Currently Released!\")")
    (else
      (return ""))))

(fn ConstructionScript)
