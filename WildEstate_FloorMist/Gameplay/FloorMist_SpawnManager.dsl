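; Server-side spawn manager that picks which FloorMist level managers activate in a match, avoiding POIs already taken by bosses.
; Once playlist data is ready, it initializes the BossSpawn_Manager and gathers POI faction tags from its spawned challenges.
; It randomly drains the level manager list, activating those with unused tags up to LevelCountToSpawn and deactivating the rest.
; It then prunes boss challenge managers on taken POIs and passes the tags back.
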(event Custom|InitializeFloorMistLevel
  (params (LevelManager :type Level%20Manager%20Floor%20Mist%20Object%20Reference))
  (if (and (not (GameplayTags|MatchesAnyTags (Variables|/WildEstate_FloorMist/Gameplay/LevelManager_FloorMist.LevelManager_FloorMist_C|GetPOIFactionTag LevelManager) (Variables|Default|GetSpawnedPOIFactionTags) true)) (< (Variables|Default|GetSpawnedLevelCount) (ScalableFloat|GetValueasInteger (Variables|Default|GetLevelCountToSpawn))))
    (Class|LevelManagerFloorMist|Init LevelManager)
    (GameplayTags|AddGameplayTag (Variables|Default|GetSpawnedPOIFactionTags) (Variables|/WildEstate_FloorMist/Gameplay/LevelManager_FloorMist.LevelManager_FloorMist_C|GetPOIFactionTag LevelManager) :!class /Script/GameplayTags.BlueprintGameplayTagLibrary)
    (Utilities|Array|Add (Variables|Default|GetActivatedLevelManagers) LevelManager)
    (Utilities|Array|RemoveItem (Variables|Default|GetLevelMangers) LevelManager)
    (Math|Integer|IncrementInt (Variables|Default|GetSpawnedLevelCount))
    (else
      (Class|LevelManagerFloorMist|DeactivateLevel LevelManager)
      (Utilities|Array|RemoveItem (Variables|Default|GetLevelMangers) LevelManager))))

(event EventBeginPlay
  (bind _spawnedpoifactiontags (Variables|Default|GetSpawnedPOIFactionTags))
  (Class|FortAsyncActionCurrentPlaylistReady|PlaylistDataReadyAsync
    (:OnReady
      (Networking|SwitchHasAuthority
        (:Authority
          (Utilities|FlowControl|DelayUntilNextTick)
          (CallFunction|ValidateBossSpawnManager)
          (Class|BossSpawnManager|InitilizewithSpawnLocationTags :self (Variables|Default|GetBossSpawnManager))
          (CallFunction|UpdateSpawnedPOIsfromBossManager)
          (for _array_element (GameplayTags|BreakGameplayTagContainer _spawnedpoifactiontags)
            (Development|PrintString (Utilities|String|Append "PreFloorMistManagerTags: " (GameplayTags|GetDebugStringfromGameplayTag _array_element)) true true "(R=0.000000,G=0.660000,B=1.000000,A=1.000000)" 2e+01))
          (CallFunction|Init _spawnedpoifactiontags)
          (CallFunction|CleanupBossSpawnManager))))))

(event Collision|EventActorBeginOverlap
  (params (OtherActor :type Actor%20Object%20Reference)))

(event EventTick
  (params (DeltaSeconds :type Float%20%28single-precision%29)))

(event Custom|ValidateBossSpawnManager
  (Variables|Default|GetBossSpawnManager
    (:then)
    (:else
      (bind _returnvalue (Actor|GetActorOfClass "/WildEstateBosses/Gameplay/BossSpawn_Manager.BossSpawn_Manager_C"))
      (Variables|Default|SetBossSpawnManager _returnvalue))))

(event Custom|UpdateSpawnedPOIsFromBossManager
  (bind _bossspawnmanager (Variables|Default|GetBossSpawnManager)
    (:then
      (for _array_element (Variables|/WildEstateBosses/Gameplay/BossSpawn_Manager.BossSpawn_Manager_C|GetSpawnedChallengeManager _bossspawnmanager)
        (bind _aschallenge_manager_simple_boss_progress_bar_wild_estate_ramdom_boss (Utilities|Casting|CastToChallengeManager_SimpleBossProgressBar_WildEstate_RamdomBoss _array_element))
        (GameplayTags|AddGameplayTag (Variables|Default|GetSpawnedPOIFactionTags) (Variables|/WildEstateLager/Gameplay/ChallengeManager_SimpleBossProgressBar_WildEstate_RamdomBoss.ChallengeManager_SimpleBossProgressBar_WildEstate_RamdomBoss_C|GetPOIFactionTag _aschallenge_manager_simple_boss_progress_bar_wild_estate_ramdom_boss) :!class /Script/GameplayTags.BlueprintGameplayTagLibrary)))
    (:else)))

(fn CleanupBossSpawnManager
  (bind _bossspawnmanager (Variables|Default|GetBossSpawnManager))
  (bind _bosschallengemanagers (Variables|/WildEstateBosses/Gameplay/BossSpawn_Manager.BossSpawn_Manager_C|GetBossChallengeManagers _bossspawnmanager))
  (LocalVariables|SetArrayLength (Utilities|Array|Length _bosschallengemanagers))
  (for _ (range (+ (LocalVariables|GetArrayLength) 1))
    (bind _aschallenge_manager_simple_boss_progress_bar_wild_estate_ramdom_boss (Utilities|Casting|CastToChallengeManager_SimpleBossProgressBar_WildEstate_RamdomBoss (Utilities|Array|Get(acopy) _bosschallengemanagers (LocalVariables|GetInt))))
    (if (GameplayTags|MatchesAnyTags (Variables|/WildEstateLager/Gameplay/ChallengeManager_SimpleBossProgressBar_WildEstate_RamdomBoss.ChallengeManager_SimpleBossProgressBar_WildEstate_RamdomBoss_C|GetPOIFactionTag _aschallenge_manager_simple_boss_progress_bar_wild_estate_ramdom_boss) (Variables|Default|GetSpawnedPOIFactionTags) true)
      (Utilities|Array|RemoveIndex (Variables|/WildEstateBosses/Gameplay/BossSpawn_Manager.BossSpawn_Manager_C|GetBossChallengeManagers (Variables|Default|GetBossSpawnManager)) (LocalVariables|GetInt))
      (else
        (Math|Integer|IncrementInt (LocalVariables|GetInt)))))
  (Variables|/WildEstateBosses/Gameplay/BossSpawn_Manager.BossSpawn_Manager_C|SetSpawnedPOILocations (Variables|Default|GetSpawnedPOIFactionTags) (Variables|Default|GetBossSpawnManager)))

(fn Init
  (params (InSpawnedPOITags :type Gameplay%20Tag%20Container%20Structure))
  (Variables|Default|SetSpawnedPOIFactionTags InSpawnedPOITags)
  (bind _outactors (Actor|GetAllActorsOfClass "/WildEstate_FloorMist/Gameplay/LevelManager_FloorMist.LevelManager_FloorMist_C"))
  (bind _output_get (Variables|Default|SetLevelMangers _outactors))
  (while (Utilities|Array|IsNotEmpty _output_get)
    (CallFunction|InitializeFloorMistLevel (Utilities|Array|RandomArrayItem (Variables|Default|GetLevelMangers))))
  (return (Variables|Default|GetSpawnedPOIFactionTags)))

(fn ConstructionScript)
