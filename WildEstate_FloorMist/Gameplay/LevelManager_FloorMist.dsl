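; FloorMist POI level manager that turns button interactions into faction changes, NPC spawns and player alerts.
; On the server, pressing a button adds the interactor's squad to the POI faction, while destroying an unpressed one skips that step.
; Both enable linked and random patrol-path providers, spawning tracked faction NPCs, and alert players through replicated pawn components.
; The area box tags entering non-faction players hostile, alerting them while spawned NPCs live after a press.
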(event Custom|BroadcastAlertEventToAllInteractorSquadMembers
  (params (InteractingPawn :type Fort%20Pawn%20Object%20Reference))
  (bind _asfort_player_pawn_athena (Utilities|Casting|CastToFortPlayerPawnAthena InteractingPawn))
  (for _array_element (CallFunction|GetPawnsfromPlayerSquadIncludingSelf _asfort_player_pawn_athena)
    (if (Utilities|IsValid (Actor|GetComponentbyClass _array_element "/WildEstate_FloorMist/Blueprints/Components/BP_FloorMistReplicatedAlertEventComponent.BP_FloorMistReplicatedAlertEventComponent_C") :!class /Script/Engine.KismetSystemLibrary)
      (Class|BPFloorMistReplicatedAlertEventComponent|BroadcastGameplayAlertEventReplicatedtoPawnOwningClient (Actor|GetComponentbyClass _array_element "/WildEstate_FloorMist/Blueprints/Components/BP_FloorMistReplicatedAlertEventComponent.BP_FloorMistReplicatedAlertEventComponent_C") (Variables|Default|GetFloorMistGameplayAlertEvent) true self))))

(event Custom|BroadcastAlertEventToAllHostirePlayersNearInteractionOrigin
  (params (InteractionOrigin :type Vector) (MaxDistance :type Float%20%28double-precision%29))
  (bind _outactors (Collision|SphereOverlapActors InteractionOrigin MaxDistance (Utilities|Array|MakeArray "ObjectTypeQuery3") "/Script/FortniteGame.FortPlayerPawnAthena"))
  (for _array_element _outactors
    (bind _asfort_player_pawn_athena (Utilities|Casting|CastToFortPlayerPawnAthena _array_element))
    (if (not (GameplayTags|HasAnyMatchingGameplayTags _asfort_player_pawn_athena (Variables|Default|GetFriendlyLevelAreaTag)))
      (Class|BPFloorMistReplicatedAlertEventComponent|BroadcastGameplayAlertEventReplicatedtoPawnOwningClient (Actor|GetComponentbyClass _asfort_player_pawn_athena "/WildEstate_FloorMist/Blueprints/Components/BP_FloorMistReplicatedAlertEventComponent.BP_FloorMistReplicatedAlertEventComponent_C") (Variables|Default|GetFloorMistGameplayAlertEvent) true self))))

(event Custom|BroadcastAlertEventToAllHostilePlayersInLevelAreaBox
  (for _array_element (Collision|GetOverlappingActors (Variables|Default|GetLevelAreaBox) "/Script/FortniteGame.FortPlayerPawnAthena" :!class /Script/Engine.PrimitiveComponent)
    (bind _asfort_player_pawn_athena (Utilities|Casting|CastToFortPlayerPawnAthena _array_element))
    (if (not (GameplayTags|HasAnyMatchingGameplayTags _asfort_player_pawn_athena (Variables|Default|GetFriendlyLevelAreaTag)))
      (Class|BPFloorMistReplicatedAlertEventComponent|BroadcastGameplayAlertEventReplicatedtoPawnOwningClient (Actor|GetComponentbyClass _asfort_player_pawn_athena "/WildEstate_FloorMist/Blueprints/Components/BP_FloorMistReplicatedAlertEventComponent.BP_FloorMistReplicatedAlertEventComponent_C") (Variables|Default|GetFloorMistGameplayAlertEvent) true self))))

(event Custom|FloorMistAddToFactions
  (params (FortPlayerPawn :type Fort%20Player%20Pawn%20Object%20Reference))
  (bind _poifactiontag (Variables|Default|GetPOIFactionTag))
  (bind _returnvalue (Game|AddComponentbyClass (Fortnite|GetFortPlayerControllerfromActor FortPlayerPawn) "/WildEstate_FloorMist/Gameplay/BP_WildEstate_FloorMist_ControllerComponent.BP_WildEstate_FloorMist_ControllerComponent_C"))
  (Class|BPWildEstateFloorMistControllerComponent|AddFactionTag _returnvalue _poifactiontag)
  (Affiliation|AddtoFaction FortPlayerPawn _poifactiontag)
  (Affiliation|ChangeFactionAttitudeTowardsActor (GameplayTags|MakeGameplayTagContainerfromTag _poifactiontag) FortPlayerPawn "Neutral")
  (Ability|GameplayEffect|RemoveLooseGameplayTags FortPlayerPawn (Variables|Default|GetHostileLevelAreaTag) true)
  (Ability|GameplayEffect|AddLooseGameplayTags FortPlayerPawn (Variables|Default|GetFriendlyLevelAreaTag) true))

(event Custom|AddFactionToTeam
  (params (InteractingPawn :type Fort%20Pawn%20Object%20Reference))
  (bind _asfort_player_pawn (Utilities|Casting|CastToFortPlayerPawn InteractingPawn))
  (CallFunction|FloorMistAddtoFactions _asfort_player_pawn)
  (bind _asfort_player_state_athena (Utilities|Casting|CastToFortPlayerStateAthena (Variables|/Script/Engine.Pawn|GetPlayerState InteractingPawn)))
  (bind _returnvalue (Class|FortGameStateAthena|GetSquadMemberPlayerStates (Fortnite|Athena|GameState|GetGameStateAthenafromContext (Fortnite|GetFortPlayerControllerfromActor InteractingPawn)) _asfort_player_state_athena false))
  (for _array_element _returnvalue
    (bind _asfort_player_pawn_2 (Utilities|Casting|CastToFortPlayerPawn (PlayerState|GetPawn _array_element)))
    (CallFunction|FloorMistAddtoFactions _asfort_player_pawn_2)))

(event Custom|ActivatePointProvider
  (params (PatrolPathPointProvider :type Fort%20Athena%20Patrol%20Path%20Point%20Provider%20Object%20Reference))
  (Utilities|IsValid PatrolPathPointProvider
    (:Is%20Valid
      (EventDispatchers|BindEventtoOnActorSpawnedResult PatrolPathPointProvider (EventDispatchers|CreateEvent :!event Custom|OnFloorMistSpawned))
      (Class|FortAthenaLivingWorldPointProviderInterface|EnablePointProvider_DEPRECATED PatrolPathPointProvider)
      ;[1 1 1] Request to spawn immediately
      (Class|FortAthenaLivingWorldManager|RequestEventGeneration (Class|FortAthenaLivingWorldManager|GetLivingWorldManager))
      (Math|Integer|IncrementInt (Variables|Default|GetCountSpawned)
        (:%20%20
          _)))
    (:Is%20Not%20Valid
      _))
  ;[1 1 1] Remove Item if in Array
  (Utilities|Array|RemoveItem (Variables|Default|GetPath%20Providers) PatrolPathPointProvider))

(event Custom|OnFloorMistSpawned
  (params (SpawnedActor :type Actor%20Object%20Reference) (bResult :type Boolean))
  (Affiliation|AddtoFaction SpawnedActor (Variables|Default|GetPOIFactionTag))
  (bind _returnvalue (Game|AddComponentbyClass SpawnedActor "/WildEstate_FloorMist/Blueprints/Components/BP_FloorMistNPCFriendlyIndicatorComponent.BP_FloorMistNPCFriendlyIndicatorComponent_C"))
  (Variables|/WildEstate_FloorMist/Blueprints/Components/BP_FloorMistNPCFriendlyIndicatorComponent.BP_FloorMistNPCFriendlyIndicatorComponent_C|SetSpawnInstigator (Variables|Default|GetLastInteractor) _returnvalue)
  (bind _asbp_player_pawn_non_participant (Utilities|Casting|CastToBP_PlayerPawn_NonParticipant SpawnedActor))
  (Utilities|Array|Add (Variables|Default|GetSpawnedNPCs) _asbp_player_pawn_non_participant)
  (Game|BindEventtoOnDestroyed _asbp_player_pawn_non_participant (EventDispatchers|CreateEvent :!event Custom|OnFloorMistDestroyed)))

(event Custom|OnFloorMistDestroyed
  (params (DestroyedActor :type Actor%20Object%20Reference))
  (bind _asbp_player_pawn_non_participant (Utilities|Casting|CastToBP_PlayerPawn_NonParticipant DestroyedActor))
  (Utilities|Array|RemoveItem (Variables|Default|GetSpawnedNPCs) _asbp_player_pawn_non_participant))

(event Custom|OnButtonDestroyed
  (params (DestroyedActor :type Actor%20Object%20Reference))
  (bind _asbuilding_actor (Utilities|Casting|CastToBuildingActor DestroyedActor))
  (if (not (Variables|Default|GetButtonPressed))
    (CallFunction|BroadcastAlertEventtoAllHostilePlayersinLevelAreaBox)
    (CallFunction|BroadcastAlertEventtoAllHostirePlayersNearInteractionOrigin (Transformation|GetActorLocation _asbuilding_actor) (Variables|Default|GetGameplayAlertBroadcastRadiusAroundButton))
    (CallFunction|SpawnButtonPaths _asbuilding_actor)
    (while (and (> (ScalableFloat|GetValueasInteger (Variables|Default|GetSpawned%20Paths%20Amount)) (Variables|Default|GetCountSpawned)) (Utilities|Array|IsNotEmpty (Variables|Default|GetPath%20Providers)))
      (CallFunction|ActivatePointProvider (Utilities|Array|RandomArrayItem (Variables|Default|GetPath%20Providers))))))

(event Custom|ButtonInteract
  (params (SelfActor :type Building%20Actor%20Object%20Reference) (InteractingPawn :type Fort%20Pawn%20Object%20Reference))
  (Variables|Default|SetLastInteractor InteractingPawn)
  (Variables|Default|SetButtonPressed true)
  (CallFunction|AddFactiontoTeam InteractingPawn)
  (CallFunction|BroadcastAlertEventtoAllInteractorSquadMembers InteractingPawn)
  (CallFunction|BroadcastAlertEventtoAllHostilePlayersinLevelAreaBox)
  (CallFunction|BroadcastAlertEventtoAllHostirePlayersNearInteractionOrigin (Transformation|GetActorLocation SelfActor) (Variables|Default|GetGameplayAlertBroadcastRadiusAroundButton))
  (CallFunction|SpawnButtonPaths SelfActor)
  (while (and (> (ScalableFloat|GetValueasInteger (Variables|Default|GetSpawned%20Paths%20Amount)) (Variables|Default|GetCountSpawned)) (Utilities|Array|IsNotEmpty (Variables|Default|GetPath%20Providers)))
    (CallFunction|ActivatePointProvider (Utilities|Array|RandomArrayItem (Variables|Default|GetPath%20Providers)))))

(event Custom|Init)

(event Custom|RemoveDuplicateButton
  (params (BottonToRemove :type Floor%20Mist%20Level%20Manager%20Button%20Point%20Provider%20Mapping%20Structure))
  (Actor|DestroyActor (Utilities|Struct|BreakFloorMistLevelManagerButtonPointProviderMapping BottonToRemove))
  (Utilities|Array|RemoveItem (Variables|Default|GetButtonsWithPatrolPaths) BottonToRemove))

(event EventBeginPlay
  (if (ScalableFloat|GetValueasBool (Variables|Default|GetEnabled))
    (Networking|SwitchHasAuthority
      (:Authority
        (Utilities|FlowControl|Sequence
          (:then_0
            (if (not (Variables|Default|GetDebug))
              _))
          (:then_1
            (for _array_element (Variables|Default|GetButtonsWithPatrolPaths)
              (Interaction|BindEventtoOnInteract (Utilities|Struct|BreakFloorMistLevelManagerButtonPointProviderMapping _array_element) (EventDispatchers|CreateEvent :!event Custom|ButtonInteract))
              (Game|BindEventtoOnDestroyed (Utilities|Struct|BreakFloorMistLevelManagerButtonPointProviderMapping _array_element) (EventDispatchers|CreateEvent :!event Custom|OnButtonDestroyed))
              (if (Variables|Default|GetDebug)
                (Development|PrintString (Utilities|String|Append "FloorMistManager: " (GameplayTags|GetDebugStringfromGameplayTag (Variables|Default|GetPOIFactionTag))) true true "(R=0.000000,G=0.660000,B=1.000000,A=1.000000)" 2e+01)))))))
    (else
      (Actor|DestroyActor))))

(event Collision|EventActorBeginOverlap
  (params (OtherActor :type Actor%20Object%20Reference)))

(event EventTick
  (params (DeltaSeconds :type Float%20%28single-precision%29)))

(event Custom|SpawnButtonPaths
  (params (InteractingButton :type Building%20Actor%20Object%20Reference))
  (bind _array_element (Utilities|Array|ForEachLoopwithBreak (Variables|Default|GetButtonsWithPatrolPaths))
    (:LoopBody
      (if (== InteractingButton (Utilities|Struct|BreakFloorMistLevelManagerButtonPointProviderMapping _array_element))
        (for _array_element_2 (.PointProvider_7_862686F348F48CCE766F81913FA28C19 (Utilities|Struct|BreakFloorMistLevelManagerButtonPointProviderMapping _array_element))
          (CallFunction|ActivatePointProvider _array_element_2))))))

(event Custom|DeactivateLevel
  (bind _buttonswithpatrolpaths (Variables|Default|GetButtonsWithPatrolPaths))
  (while (Utilities|Array|IsNotEmpty _buttonswithpatrolpaths)
    (CallFunction|RemoveDuplicateButton (Utilities|Array|RandomArrayItem _buttonswithpatrolpaths))
    (Components|DestroyComponent (Variables|Default|GetLevelAreaBox))))

(event OnComponentBeginOverlap(LevelAreaBox)
  (params (OverlappedComponent :type Primitive%20Component%20Object%20Reference) (OtherActor :type Actor%20Object%20Reference) (OtherComp :type Primitive%20Component%20Object%20Reference) (OtherBodyIndex :type Integer) (bFromSweep :type Boolean) (SweepResult :type Hit%20Result%20Structure%20%28by%20ref%29))
  (bind _asfort_player_pawn_athena (Utilities|Casting|CastToFortPlayerPawnAthena OtherActor))
  (if (and (Pawn|IsPlayerControlled _asfort_player_pawn_athena) (not (GameplayTags|HasMatchingGameplayTag _asfort_player_pawn_athena (Variables|Default|GetPOIFactionTag))))
    (Ability|GameplayEffect|AddLooseGameplayTags _asfort_player_pawn_athena (Variables|Default|GetHostileLevelAreaTag) true)
    ;[0.15 0.04141 0.04141 0.5] Broadcast Hostile Alert If Zone Active
    (if (and (Variables|Default|GetButtonPressed) (CallFunction|AreAnySpawnedNPCsAlive))
      (if (Utilities|IsValid (Actor|GetComponentbyClass _asfort_player_pawn_athena "/WildEstate_FloorMist/Blueprints/Components/BP_FloorMistReplicatedAlertEventComponent.BP_FloorMistReplicatedAlertEventComponent_C") :!class /Script/Engine.KismetSystemLibrary)
        (Class|BPFloorMistReplicatedAlertEventComponent|BroadcastGameplayAlertEventReplicatedtoPawnOwningClient (Actor|GetComponentbyClass _asfort_player_pawn_athena "/WildEstate_FloorMist/Blueprints/Components/BP_FloorMistReplicatedAlertEventComponent.BP_FloorMistReplicatedAlertEventComponent_C") (Variables|Default|GetFloorMistGameplayAlertEvent) true self)))))

(event OnComponentEndOverlap(LevelAreaBox)
  (params (OverlappedComponent :type Primitive%20Component%20Object%20Reference) (OtherActor :type Actor%20Object%20Reference) (OtherComp :type Primitive%20Component%20Object%20Reference) (OtherBodyIndex :type Integer))
  (bind _asfort_player_pawn_athena (Utilities|Casting|CastToFortPlayerPawnAthena OtherActor))
  (if (Pawn|IsPlayerControlled _asfort_player_pawn_athena)
    (Ability|GameplayEffect|RemoveLooseGameplayTags _asfort_player_pawn_athena (Variables|Default|GetHostileLevelAreaTag) true)
    (Ability|GameplayEffect|RemoveLooseGameplayTags _asfort_player_pawn_athena (Variables|Default|GetFriendlyLevelAreaTag) true)))

(fn AreAnySpawnedNPCsAlive
  (for _array_element (Variables|Default|GetSpawnedNPCs)
    (if (Utilities|IsValid _array_element :!class /Script/Engine.KismetSystemLibrary)
      (if (not (FortPawn|IsDead _array_element))
        (return true))))
  (return false))

(fn GetPawnsFromPlayerSquadIncludingSelf
  (params (PlayerPawn :type Fort%20Player%20Pawn%20Athena%20Object%20Reference))
  (bind _asfort_player_state_athena (Utilities|Casting|CastToFortPlayerStateAthena (Variables|/Script/Engine.Pawn|GetPlayerState PlayerPawn)))
  (bind _returnvalue (Class|FortGameStateAthena|GetSquadMemberPlayerStates (Fortnite|Athena|GameState|GetGameStateAthenafromContext (Fortnite|GetFortPlayerControllerfromActor PlayerPawn)) _asfort_player_state_athena true))
  (for _array_element _returnvalue
    (bind _asfort_player_pawn_athena (Utilities|Casting|CastToFortPlayerPawnAthena (PlayerState|GetCurrentPawn _array_element)))
    (Utilities|Array|Add (LocalVariables|GetTempArray_PlayerSquadPawns) _asfort_player_pawn_athena))
  (return (LocalVariables|GetTempArray_PlayerSquadPawns)))

(fn ConstructionScript)
